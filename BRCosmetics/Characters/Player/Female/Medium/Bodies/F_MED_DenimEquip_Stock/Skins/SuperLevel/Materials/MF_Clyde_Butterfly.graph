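# Material function for the Clyde Butterfly skin that adds an animated galaxy look to incoming material attributes and blends the result back.
# A scrolling nebula texture is projected through a triplanar virtual box for depth, then hue, brightness and contrast adjusted, with time-of-day brightness and desaturation.
# Channels of one mask texture drive galaxy coverage, emissive and alpha, alongside a colored Fresnel glow.
# A parameter blends galaxy surface values with incoming ones.
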MaterialFunction /BRCosmetics/Characters/Player/Female/Medium/Bodies/F_MED_DenimEquip_Stock/Skins/SuperLevel/Materials/MF_Clyde_Butterfly
graph {nodes [
  Constant()
  FunctionOutput(None=BlendMaterialAttributes(A=NamedRerouteUsage(), B=NamedRerouteUsage(), Alpha=NamedRerouteUsage()))
  n1 = Fresnel(ExponentIn=ScalarParameter() {defaultValue=3, group=Galaxy, parameterName=EmissiveFresnelExponent, sortPriority=90}) {baseReflectFraction=0}
  n2 = ScalarParameter() {defaultValue=1, group=Galaxy, parameterName=EmissiveFresnelIntensity, sortPriority=92}
  n3 = NamedRerouteUsage()
  n4 = Power(Base=NamedRerouteUsage(), Exp=ScalarParameter() {defaultValue=1, group=Galaxy, parameterName=EmissiveMaskExponent, sortPriority=93})
  n5 = Saturate(None=Add(A=ComponentMask(None=@n3) {g=true}, B=@n4))
  n6 = ScalarParameter() {group=Galaxy, parameterName=UseEmissiveMaskOnFresnel, sortPriority=92}
  n7 = NamedRerouteDeclaration(None=Multiply(A=Multiply(A=@n1, B=@n2), B=LinearInterpolate(B=@n5, Alpha=@n6) {constA=1})) {name=FresnelMask, nodeColor=<rgba 0 0.7891 1 1>}
  n8 = VectorParameter() {group=Galaxy, parameterName=EmissiveColor, sortPriority=91, defaultValue=<rgba 0.0885 0.6921 1 1>}
  n9 = Multiply(A=@n4, B=ScalarParameter() {group=Galaxy, parameterName=EmissiveMaskIntensity, sortPriority=92})
  n10 = StaticSwitchParameter(True=NamedRerouteUsage(), False=@n8.RGB) {group=Galaxy, parameterName=UseEmissiveMap, sortPriority=89}
  NamedRerouteDeclaration(None=Add(A=Multiply(A=@n7, B=@n8.RGB), B=Multiply(A=@n9, B=@n10))) {name=Fresnel, nodeColor=<rgba 0.8047 0 1 1>}
  NamedRerouteDeclaration(None=Saturate(None=NamedRerouteUsage())) {name=Galaxies, nodeColor=<rgba 1 0 0.2578 1>}
  n11 = FunctionInput() {inputName=Material, inputType=FunctionInput_MaterialAttributes}
  n12 = GetMaterialAttributes(None=@n11) {attributeGetTypes [
    #1
    #2
    #3
    #4
    #5
   ]}
  n13 = NamedRerouteDeclaration(None=ScalarParameter() {group=Galaxy, parameterName=UseIncomingMatParams}) {name=UseIncomingMatParams, nodeColor=<rgba 0.875 0 1 1>}
  NamedRerouteDeclaration(None=LinearInterpolate(A=ScalarParameter() {defaultValue=0.2, group=Galaxy, parameterName=Galaxy_Roughness, sortPriority=52}, B=@n12.Roughness, Alpha=@n13) {constB=0}) {name=OUTRoughness, nodeColor=<rgba 0 0.4844 1 1>}
  NamedRerouteDeclaration(None=LinearInterpolate(A=ScalarParameter() {defaultValue=0.1, group=Galaxy, parameterName=Galaxy_Metallic, sortPriority=50}, B=@n12.Metallic, Alpha=@n13) {constB=0}) {name=OUTMetallic, nodeColor=<rgba 0 0.4844 1 1>}
  n14 = TextureObjectParameter() {
   group=Galaxy
   parameterName=Galaxy_01_Texture
   sortPriority=58
   texture=/Game/Characters/Player/Male/Medium/Bodies/M_Med_Soldier_04/Skins/Galaxy/Textures/T_FN_Nebula_style_blue
  }
  n15 = ScalarParameter() {defaultValue=80, group=Galaxy, parameterName=Galaxy_01_Depth, sortPriority=61}
  n16 = ScalarParameter() {defaultValue=0.4, group=Galaxy, parameterName=Galaxy_01_Tile, sortPriority=62}
  n17 = ScalarParameter() {defaultValue=0.4, group=Galaxy, parameterName=Galaxy_01_BlendContrast, sortPriority=63}
  n18 = VectorParameter() {group=Galaxy, parameterName=Galaxy_01_Speed, sortPriority=64, defaultValue=<rgba 0.002 -0.0025 0 1>}
  n19 = ComponentMask(None=@n18.RGB) {g=true, r=true}
  n20 = NamedRerouteDeclaration(None=VectorParameter() {group=Galaxy, parameterName=Galaxy_Offset, sortPriority=59, defaultValue=<rgba 0 0 0 1>}.RGB) {name=ActorOffset, nodeColor=<rgba 0.9688 1 0 1>}
  n21 = Multiply(
   A=MaterialFunctionCall(
    Hue Shift Percentage=ScalarParameter() {group=Galaxy, parameterName=Galaxy_01_HueShift, sortPriority=64}
    Texture=MaterialFunctionCall!(Texture=@n14, Depth=@n15, Tile=@n16, BlendContrast=@n17, SpeedXY=@n19, ActorOffset=@n20, UseLocalCameraVector) {materialFunction=/Game/Global/MaterialFunctions/MF_TriplanarVirtualBox}.Result
   ) {materialFunction=&HueShift}.Result
   B=ScalarParameter() {defaultValue=1, group=Galaxy, parameterName=Galaxy_01_Brightness, sortPriority=65}
  )
  n22 = ScalarParameter() {defaultValue=1.75, group=Galaxy, parameterName=Galaxy_01_Exp, sortPriority=66}
  NamedRerouteDeclaration(None=Desaturation(
    None=Power(Base=@n21, Exp=@n22)
    Fraction=ScalarParameter() {defaultValue=0.5, group=Galaxy, parameterName=Galaxy_01_Desaturation, sortPriority=67}
   )) {name=Galaxy_01, nodeColor=<rgba 1 0.2578 0 1>}
  NamedRerouteDeclaration(None=LinearInterpolate(A=ScalarParameter() {defaultValue=0.5, group=Galaxy, parameterName=Galaxy_Spec, sortPriority=51}, B=@n12.Specular, Alpha=@n13) {constB=0}) {name=OUTSpec, nodeColor=<rgba 0 0.4844 1 1>}
  NamedRerouteDeclaration(None=@n12.EmissiveColor) {name=IN_Emissive, nodeColor=<rgba 0.4766 1 0 1>}
  NamedRerouteDeclaration(None=@n10) {name=FresnelEmissive, nodeColor=<rgba 0 1 0.7656 1>}
  n23 = TextureSampleParameter2D() {
   group=Galaxy
   parameterName=MaskTexture_DetailedControl
   samplerSource=SSM_Wrap_WorldGroupSettings
   samplerType=SAMPLERTYPE_Masks
   sortPriority=-20
   texture=&T_WhiteMask
  }
  n24 = ChannelMaskParameter(None=@n23.RGB) {group=Galaxy, maskChannel=Blue, parameterName=GalaxyMaskChannel, sortPriority=-7, defaultValue=<rgba 0 0 1 0>}
  NamedRerouteDeclaration(None=LinearInterpolate(B=@n24.R, Alpha=ScalarParameter() {defaultValue=1, group=Galaxy, parameterName=GalaxyMaskStrength, sortPriority=-6}) {constA=1}) {name=GalaxyMask, nodeColor=<rgba 0.3125 0 1 1>}
  n25 = Multiply(A=OneMinus(None=NamedRerouteUsage()), B=NamedRerouteUsage())
  NamedRerouteDeclaration(None=Add(A=@n25, B=NamedRerouteUsage())) {name=OUT_Emissive, nodeColor=<rgba 1 0 0.1875 1>}
  NamedRerouteDeclaration(None=LinearInterpolate(
    A=VectorParameter() {group=Galaxy, parameterName=Galaxy_BaseColor, sortPriority=49, defaultValue=<rgba 0 0 0 1>}.RGB
    B=@n12.EmissiveColor
    Alpha=@n13
   ) {constB=0}) {name=OUTBaseColor, nodeColor=<rgba 1 0 0.6562 1>}
  n26 = FunctionInput() {bUsePreviewValueAsDefault=true, inputName=TechArtMask, sortPriority=1, previewValue=<1 1 1 1>}
  NamedRerouteDeclaration(None=@n26) {name=TechArtMask, nodeColor=<rgba 0 1 0.6953 1>}
  n27 = SetMaterialAttributes!(
   MaterialAttributes=NamedRerouteUsage()
   BaseColor=NamedRerouteUsage()
   Metallic=NamedRerouteUsage()
   Specular=NamedRerouteUsage()
   Roughness=NamedRerouteUsage()
   EmissiveColor=NamedRerouteUsage()
  ) {attributeSetTypes [
    #5
    #2
    #3
    #1
    #4
   ]}
  NamedRerouteDeclaration(None=@n27) {name=MaterialOUT, nodeColor=<rgba 0 1 0.625 1>}
  NamedRerouteDeclaration(None=ChannelMaskParameter(None=@n23.RGB) {group=Galaxy, parameterName=AlphaChannel, sortPriority=-10}.R) {name=Alpha, nodeColor=<rgba 0.7109 1 0 1>}
  NamedRerouteDeclaration(None=@n11) {name=IN_Material, nodeColor=<rgba 1 0 0.9609 1>}
  n28 = NamedRerouteUsage()
  n29 = MaterialFunctionCall(
   Value_Morning=ScalarParameter() {defaultValue=1, group=Galaxy, parameterName=Galaxy_Brightness_Morning, sortPriority=100}
   Value_Day=ScalarParameter() {defaultValue=1, group=Galaxy, parameterName=Galaxy_Brightness_Day, sortPriority=101}
   Value_Evening=ScalarParameter() {defaultValue=1, group=Galaxy, parameterName=Galaxy_Brightness_Evening, sortPriority=102}
   Value_Night=ScalarParameter() {defaultValue=1, group=Galaxy, parameterName=Galaxy_Brightness_Night, sortPriority=103}
  ) {materialFunction=&Scalar_Time_of_Day}
  n30 = ScalarParameter() {group=Galaxy, parameterName=Galaxy_Desat_Morning, sortPriority=110}
  n31 = ScalarParameter() {group=Galaxy, parameterName=Galaxy_Desat_Day, sortPriority=111}
  n32 = ScalarParameter() {group=Galaxy, parameterName=Galaxy_Desat_Evening, sortPriority=112}
  n33 = ScalarParameter() {group=Galaxy, parameterName=Galaxy_Desat_Night, sortPriority=113}
  NamedRerouteDeclaration(None=Desaturation(
    None=Multiply(A=@n28, B=@n29.Result)
    Fraction=MaterialFunctionCall(Value_Morning=@n30, Value_Day=@n31, Value_Evening=@n32, Value_Night=@n33) {materialFunction=&Scalar_Time_of_Day}.Result
   )) {name=GalaxiesAndStars, nodeColor=<rgba 0 1 0.9063 1>}
  n34 = ChannelMaskParameter(None=@n23.RGB) {group=Galaxy, maskChannel=Green, parameterName=EmissiveMaskChannel, sortPriority=-9, defaultValue=<rgba 0 1 0 0>}
  NamedRerouteDeclaration(None=@n34.R) {name=EmissiveMask, nodeColor=<rgba 0 0.4375 1 1>}
  StaticSwitchParameter(True=Multiply(A=NamedRerouteUsage(), B=@n25), False=@n25) {group=Galaxy, parameterName=UseGalaxyMask, sortPriority=-8}
 ]}
