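# Parent character material for Clyde SuperLevel skins, layering animated paint, accent, butterfly and graffiti effects between the 2023 CharacterParent PreFX and PostFX functions.
# Mask channels blend noise-warped paint splotches, metallic accent color and a butterfly-function layer over the base.
# Butterflies use camera-facing virtual planes at chest and hand positions, switchable to BP-supplied parameters, with sphere-masked glow.
# A custom HLSL loop composites five colors of seeded, animated flipbook graffiti into base color.
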Material /BRCosmetics/Characters/Player/Female/Medium/Bodies/F_MED_DenimEquip_Stock/Skins/SuperLevel/Materials/M_ClydeSuperlevels_PARENT
graph {
 properties {
  bUseMaterialAttributes=true
  bUsedWithMeshDeformer=true
  bUsedWithMorphTargets=true
  bUsedWithSkeletalMesh=true
  pixelDepthOffsetMode=PDOM_Legacy
  shadingModel=MSM_FromMaterialExpression
  twoSided=true
  usedShadingModels=Default Lit | Subsurface | Clear Coat | Subsurface Profile | Cloth
 }
 nodes [
  Constant()
  Constant()
  n1 = NamedRerouteUsage()
  n2 = GetMaterialAttributes(None=@n1) {attributeGetTypes [#1, #2]}
  n3 = NamedRerouteUsage()
  n4 = OneMinus(None=@n3)
  n5 = Add(A=@n3) {constB=0.5}
  n6 = SetMaterialAttributes!(
   MaterialAttributes=@n2.MaterialAttributes
   Base Color=Multiply(A=Desaturation(None=@n2.Base Color, Fraction=@n4), B=Saturate(None=@n5)) {constB=0.5}
   Metallic=Min(A=@n2.Metallic) {constB=0}
  ) {attributeSetTypes [#1, #2]}
  n7 = TextureSampleParameter2D() {
   group=SUPERLEVEL
   parameterName=PaintTexture
   sortPriority=60
   texture=/BRCosmetics/Characters/Player/Female/Medium/Bodies/F_MED_DenimEquip_Stock/Skins/SuperLevel/Textures/T_DenimEquip_Mask_Paint_SL
  }
  n8 = Add(A=@n7.R, B=@n7.R)
  n9 = Saturate(None=Multiply(A=Add(A=@n8, B=@n7.B)) {constB=2})
  n10 = NamedRerouteUsage()
  n11 = TextureSampleParameter2D(UVs=Panner(Coordinate=Multiply(B=@n10) {constA=0.2}) {speedX=0.013, speedY=-0.025}) {parameterName=Texture, samplerSource=SSM_Wrap_WorldGroupSettings, texture=&AtmosphericCloudNoise01}
  n12 = Multiply(A=@n11.R) {constB=0.0085}
  n13 = Subtract(
   A=ComponentMask(None=NamedRerouteUsage()) {b=true}
   B=ScalarParameter() {defaultValue=1000, group=SUPERLEVEL, parameterName=Splotches_WarpMaskHeight, sortPriority=20}
  )
  n14 = NamedRerouteDeclaration(None=OneMinus(None=Saturate(None=@n13))) {name=Splotches_WarpMask, nodeColor=<rgba 0.5938 0 1 1>}
  n15 = TextureSampleParameter2D(UVs=Add(A=TextureCoordinate(), B=Multiply(A=@n12, B=@n14))) {
   group=SUPERLEVEL
   parameterName=PaintTexture
   sortPriority=60
   texture=&T_DenimEquip_Mask_Paint_SL
  }
  n16 = SetMaterialAttributes!(MaterialAttributes=@n1, Base Color=Max(A=Multiply(A=@n9, B=@n15.RGB), B=Constant() {r=0.03}), Roughness, Metallic) {attributeSetTypes [#1, #3, #2]}
  n17 = NamedRerouteUsage()
  n18 = Constant() {r=1}
  n19 = VectorParameter() {group=SUPERLEVEL, parameterName=AccentColor, defaultValue=<rgba 0 0.1704 0.3802 1>}
  n20 = NamedRerouteUsage()
  n21 = Constant() {r=0.05}
  n22 = NamedRerouteUsage()
  n23 = NamedRerouteUsage()
  n24 = Add(A=Add(A=Add(A=@n22, B=@n23), B=NamedRerouteUsage()), B=NamedRerouteUsage())
  n25 = BlendMaterialAttributes(
   A=BlendMaterialAttributes(
    A=BlendMaterialAttributes(A=@n6, B=@n16, Alpha=@n17)
    B=SetMaterialAttributes!(MaterialAttributes, Metallic=@n18, Base Color=@n19.RGB) {attributeSetTypes [#2, #1]}
    Alpha=NamedRerouteUsage()
   )
   B=StaticSwitchParameter(
    True=@n20
    False=SetMaterialAttributes!(MaterialAttributes=@n20, Base Color=@n21, Specular) {attributeSetTypes [#1, #4]}
   ) {group=SUPERLEVEL, parameterName=IsStageCue_FaceAcc, sortPriority=120}
   Alpha=OneMinus(None=Saturate(None=@n24))
  )
  n26 = MaterialFunctionCall!(Material=@n25, TechArtMask) {materialFunction=&/MF_Clyde_Butterfly}
  n27 = GetMaterialAttributes(None=@n26.Result) {attributeGetTypes [#5]}
  n28 = NamedRerouteUsage()
  n29 = MaterialFunctionCall(Hue Shift Percentage=ScalarParameter() {group=SUPERLEVEL, parameterName=Galaxy_HueShift}, Texture=@n27.Emissive Color) {materialFunction=&HueShift}
  n30 = ScalarParameter() {group=SUPERLEVEL, parameterName=Galaxy_Desaturation}
  n31 = LinearInterpolate(
   A=Multiply(A=Desaturation(None=@n29.Result, Fraction=@n30), B=ScalarParameter() {defaultValue=1, group=SUPERLEVEL, parameterName=Galaxy_Brightness})
   B=NamedRerouteUsage()
   Alpha=NamedRerouteUsage()
  )
  n32 = Add(
   A=NamedRerouteUsage()
   B=Multiply(
    A=MaterialFunctionCall(Hue Shift Percentage=@n28, Texture=@n31) {materialFunction=&HueShift}.Result
    B=NamedRerouteUsage()
   )
  )
  n33 = NamedRerouteUsage()
  n34 = NamedRerouteUsage()
  n35 = Multiply(A=@n34, B=NamedRerouteUsage())
  n36 = ScalarParameter() {group=SUPERLEVEL, parameterName=ButterflyOverTheClothes_CenterGlowIntensity}
  n37 = NamedRerouteUsage()
  n38 = Add(A=Add(A=NamedRerouteUsage(), B=NamedRerouteUsage()), B=@n37)
  n39 = Add(A=OneMinus(None=Saturate(None=@n38)), B=@n37)
  n40 = Multiply(
   A=Add(A=@n34, B=Multiply(A=@n35, B=@n36))
   B=Saturate(None=StaticSwitchParameter(True=@n39, False=@n37) {group=SUPERLEVEL, parameterName=OverTheClothesButterfly_UseAdditionalMask})
  )
  n41 = ScalarParameter() {defaultValue=0.065, group=SUPERLEVEL, parameterName=ButterflyOverTheClothes_OpacityMultiplier}
  n42 = Add(A=NamedRerouteUsage(), B=Multiply(A=OneMinus(None=@n33), B=Multiply(A=@n40, B=@n41) {constB=0.065}))
  n43 = GetMaterialAttributes(None=BlendMaterialAttributes(
    A=@n25
    B=SetMaterialAttributes!(MaterialAttributes=@n27.MaterialAttributes, Emissive Color=@n32) {attributeSetTypes [#5]}
    Alpha=Saturate(None=@n42)
   )) {attributeGetTypes [#1, #2, #4]}
  MaterialExpressionTextureObjectParameter_0 = TextureObjectParameter() {
   group=SUPERLEVEL
   parameterName=GraffitiFlipbookTexture
   samplerType=SAMPLERTYPE_Masks
   sortPriority=-1000
   texture=/BRCosmetics/Characters/Player/Female/Medium/Bodies/F_MED_DenimEquip_Stock/Skins/SuperLevel/Textures/T_GraffitiFlipbook_Face
  }
  MaterialExpressionConstant_16 = Constant() {r=4}
  MaterialExpressionTime_7 = Time()
  MaterialExpressionNamedRerouteUsage_11 = NamedRerouteUsage()
  MaterialExpressionNamedRerouteUsage_70 = NamedRerouteUsage()
  n44 = VectorParameter() {group=SUPERLEVEL, parameterName=GraffitiColor_01, sortPriority=40, defaultValue=<rgba 1 0.6873 0 1>}
  n45 = Constant() {r=1}
  MaterialExpressionAppendVector_4 = AppendVector(A=@n44.RGB, B=@n45)
  n46 = VectorParameter() {group=SUPERLEVEL, parameterName=GraffitiColor_02, sortPriority=42, defaultValue=<rgba 1 0 0.4605 1>}
  MaterialExpressionAppendVector_5 = AppendVector(A=@n46.RGB, B=@n45)
  n47 = VectorParameter() {group=SUPERLEVEL, parameterName=GraffitiColor_03, sortPriority=44, defaultValue=<rgba 0 1 0.1982 1>}
  MaterialExpressionAppendVector_6 = AppendVector(A=@n47.RGB, B=@n45)
  n48 = VectorParameter() {group=SUPERLEVEL, parameterName=GraffitiColor_04, sortPriority=46, defaultValue=<rgba 0.9583 1 0.9167 1>}
  MaterialExpressionAppendVector_7 = AppendVector(A=@n48.RGB, B=@n45)
  n49 = VectorParameter() {group=SUPERLEVEL, parameterName=GraffitiColor_05, sortPriority=48, defaultValue=<rgba 0 0.9878 1 1>}
  MaterialExpressionAppendVector_8 = AppendVector(A=@n49.RGB, B=@n45)
  MaterialExpressionNamedRerouteUsage_71 = NamedRerouteUsage()
  MaterialExpressionNamedRerouteUsage_72 = NamedRerouteUsage()
  MaterialExpressionNamedRerouteUsage_73 = NamedRerouteUsage()
  MaterialExpressionNamedRerouteUsage_74 = NamedRerouteUsage()
  MaterialExpressionNamedRerouteUsage_75 = NamedRerouteUsage()
  MaterialExpressionNamedRerouteUsage_76 = NamedRerouteUsage()
  MaterialExpressionNamedRerouteUsage_91 = NamedRerouteUsage()
  MaterialExpressionNamedRerouteUsage_77 = NamedRerouteUsage()
  MaterialExpressionNamedRerouteUsage_78 = NamedRerouteUsage()
  MaterialExpressionNamedRerouteUsage_79 = NamedRerouteUsage()
  MaterialExpressionNamedRerouteUsage_80 = NamedRerouteUsage()
  MaterialExpressionNamedRerouteUsage_81 = NamedRerouteUsage()
  MaterialExpressionNamedRerouteUsage_89 = NamedRerouteUsage()
  MaterialExpressionNamedRerouteUsage_82 = NamedRerouteUsage()
  MaterialExpressionNamedRerouteUsage_90 = NamedRerouteUsage()
  MaterialExpressionNamedRerouteUsage_83 = NamedRerouteUsage()
  MaterialExpressionNamedRerouteUsage_84 = NamedRerouteUsage()
  MaterialExpressionNamedRerouteUsage_85 = NamedRerouteUsage()
  MaterialExpressionScalarParameter_12 = ScalarParameter() {defaultValue=0.1, group=SUPERLEVEL, parameterName=GrafittiFace_LoopSpeed}
  MaterialExpressionMultiply_54 = Multiply(A=ScalarParameter() {defaultValue=1, group=SUPERLEVEL, parameterName=GrafittiFace_RevealDuration}) {constB=1.78}
  n50 = VectorParameter() {group=SUPERLEVEL, parameterName=FlipbookSeed, defaultValue=<rgba 124.391 62.798 632.891 195.321>}
  MaterialExpressionMaterialFunctionCall_23 = MaterialFunctionCall(R=@n50.R, G=@n50.R, B=@n50.R, A=@n50.R) {materialFunction=&AppendMany}
  MaterialExpressionNamedRerouteUsage_88 = NamedRerouteUsage()
  n51 = Custom!(
   inputTex=@MaterialExpressionTextureObjectParameter_0
   rows=@MaterialExpressionConstant_16
   inputTime=@MaterialExpressionTime_7
   inputUVs=@MaterialExpressionNamedRerouteUsage_11
   inputUVs02=@MaterialExpressionNamedRerouteUsage_70
   graffitiColor01=@MaterialExpressionAppendVector_4
   graffitiColor02=@MaterialExpressionAppendVector_5
   graffitiColor03=@MaterialExpressionAppendVector_6
   graffitiColor04=@MaterialExpressionAppendVector_7
   graffitiColor05=@MaterialExpressionAppendVector_8
   uvOffset01Arrow=@MaterialExpressionNamedRerouteUsage_71
   uvOffset02Arrow=@MaterialExpressionNamedRerouteUsage_72
   uvOffset03Arrow=@MaterialExpressionNamedRerouteUsage_73
   uvOffset04Arrow=@MaterialExpressionNamedRerouteUsage_74
   uvOffset05Arrow=@MaterialExpressionNamedRerouteUsage_75
   uvOffset01Flower=@MaterialExpressionNamedRerouteUsage_76
   uvOffset01FlowerBack=@MaterialExpressionNamedRerouteUsage_91
   uvOffset02Flower=@MaterialExpressionNamedRerouteUsage_77
   uvOffset03Flower=@MaterialExpressionNamedRerouteUsage_78
   uvOffset04Flower=@MaterialExpressionNamedRerouteUsage_79
   uvOffset05Flower=@MaterialExpressionNamedRerouteUsage_80
   uvOffset01Crown=@MaterialExpressionNamedRerouteUsage_81
   uvOffset01CrownBack=@MaterialExpressionNamedRerouteUsage_89
   uvOffset02Crown=@MaterialExpressionNamedRerouteUsage_82
   uvOffset02CrownBack=@MaterialExpressionNamedRerouteUsage_90
   uvOffset03Crown=@MaterialExpressionNamedRerouteUsage_83
   uvOffset04Crown=@MaterialExpressionNamedRerouteUsage_84
   uvOffset05Crown=@MaterialExpressionNamedRerouteUsage_85
   loopSpeed=@MaterialExpressionScalarParameter_12
   revealDuration=@MaterialExpressionMultiply_54
   seeds=@MaterialExpressionMaterialFunctionCall_23.RGBA
   graffitiFrontMask=@MaterialExpressionNamedRerouteUsage_88
  ) {
   code="float4 result = float4(0,0,0,0);\r\nfloat2 seed01 = float2(1279.643, 8482.2315);\r\nfloat2 seed02 = float2(765.421, 3275.1234);\r\nfloat2 seed03 = float2(849.583, 2215.306);\r\nfloat2 offsetRange = float2(0,1);\r\n\r\n\r\n\r\nfor(int i = 0; i < 5; i++)\r\n{\r\n\tfloat4 graffitiColor = float4(0,0,0,0); \r\n    float4 offsetArrow = float4(0,0,0,1);\r\n    float4 offsetFlower = float4(0,0,0,1);\r\n    float4 offsetCrown = float4(0,0,0,1);\r\n\r\n    float2 uvCoords = inputUVs;\r\n\r\nif(i==0){\r\n      graffitiColor = graffitiColor01;\r\n      offsetArrow = uvOffset01Arrow;\r\n      offsetFlower = lerp(uvOffset01Flower, uvOffset01FlowerBack, 1-graffitiFrontMask);\r\n      offsetCrown = lerp(uvOffset01Crown, uvOffset01CrownBack, 1-graffitiFrontMask);\r\n\r\n    }\r\n\r\nif(i==1){\r\n      graffitiColor = graffitiColor02;\r\n      offsetArrow = uvOffset02Arrow;\r\n      offsetFlower = uvOffset02Flower;\r\n      offsetCrown = lerp(uvOffset02Crown, uvOffset02CrownBack, 1-graffitiFrontMask);\r\n}\r\n\r\nif(i==2){\r\n      graffitiColor = graffitiColor03;\r\n      offsetArrow = uvOffset03Arrow;\r\n      offsetFlower = uvOffset03Flower;\r\n      offsetCrown = uvOffset03Crown;}\r\n\r\nif(i==3){\r\n      graffitiColor = graffitiColor04;\r\n      offsetArrow = uvOffset04Arrow;\r\n      offsetFlower = uvOffset04Flower;\r\n      offsetCrown = uvOffset04Crown;\r\n\r\n      uvCoords = inputUVs02; //this uv is projected from the sides rather than from the front\r\n}\r\n\r\nif(i==4){\r\n      graffitiColor = graffitiColor05;\r\n      offsetArrow = uvOffset05Arrow;\r\n      offsetFlower = uvOffset05Flower;\r\n      offsetCrown = uvOffset05Crown;\r\n\r\n      uvCoords = inputUVs02; //this uv is projected from the sides rather than from the front\r\n}\r\n\r\n    seed01 = frac(seed01*float2(seeds.x, seeds.y));\r\n    seed02 = frac(seed02*float2(seeds.z, seeds.w));\r\n\t\r\n\r\n\tfloat randScale01 = 0.5*offsetArrow.w;\r\n    float randScale02 = 0.5*offsetFlower.w;\r\n    float randScale03 = 0.5*offsetCrown.w;\r\n\t\r\n\t\r\n\tfloat2x2 rotationMatrix01 = float2x2(cos(offsetArrow.z), -sin(offsetArrow.z),\r\n\t\t\t\t\t\t\t\t\t   sin(offsetArrow.z), cos(offsetArrow.z));\r\n\r\n    float2x2 rotationMatrix02 = float2x2(cos(offsetFlower.z), -sin(offsetFlower.z),\r\n\t\t\t\t\t\t\t\t\t   sin(offsetFlower.z), cos(offsetFlower.z));\r\n\r\n    float2x2 rotationMatrix03 = float2x2(cos(offsetCrown.z), -sin(offsetCrown.z),\r\n\t\t\t\t\t\t\t\t\t   sin(offsetCrown.z), cos(offsetCrown.z));\r\n\r\n    float2 randOffset01 = mul(rotationMatrix01, float2(offsetArrow.x, offsetArrow.y)*(1/randScale01)); \r\n\tfloat2 randOffset02 = mul(rotationMatrix02, float2(offsetFlower.x, offsetFlower.y)*(1/randScale02));\r\n    float2 randOffset03 = mul(rotationMatrix03, float2(offsetCrown.x, offsetCrown.y)*(1/randScale03));\r\n\t\t\t\t\t\t\t\t  \r\nfloat2 uvResult = mul(rotationMatrix01, ((uvCoords/randScale01)+0.5) - (0.5/randScale01));\t\r\nfloat2 uvResult02 = mul(rotationMatrix02, ((uvCoords/randScale02)+0.5) - (0.5/randScale02));\r\nfloat2 uvResult03 = mul(rotationMatrix03, ((uvCoords/randScale03)+0.5) - (0.5/randScale03));\r\n\r\n\r\n\r\n\r\nfloat randomDuration01 = lerp(revealDuration*0.7, revealDuration, seed01.y);\r\nfloat randomDuration02 = lerp(revealDuration*1.05, revealDuration*0.64, seed02.x);\r\nfloat randomDuration03 = lerp(revealDuration*0.8, revealDuration*0.93, seed02.y);\r\nfloat newTime01 = min(smoothstep(0,loopSpeed*randomDuration01,frac((((inputTime+9*seed01.x)+0.5)*loopSpeed))), 0.96);\r\nfloat newTime02 = min(smoothstep(0,loopSpeed*randomDuration02,frac(((inputTime+14*seed02.y +0.2)*loopSpeed))), 0.98);\r\nfloat newTime03 = min(smoothstep(0,loopSpeed*randomDuration03,frac(((inputTime+8*seed01.y)*loopSpeed))), 0.97);\r\n\r\nfloat2 animatedUV01 = (((uvResult/rows) + floor(float2(rows*rows,rows)*frac(newTime01))/rows)) + randOffset01 + (sin(inputTime*6.2+(11*seed02.y))*0.01);\r\nfloat2 staticUV01 = (uvResult/rows) + randOffset01 + (sin(inputTime*6.2+(11*seed02.y))*0.01);\r\nfloat2 animatedUV02 = (((uvResult02/rows) + floor(float2(rows*rows,rows)*frac(newTime02))/rows)) + randOffset02 + (sin(inputTime*9+(14*seed02.x))*0.00485);\r\nfloat2 staticUV02 = (uvResult02/rows) + randOffset02 + (sin(inputTime*9+(14*seed02.x))*0.0065);\r\nfloat2 animatedUV03 = (((uvResult03/rows) + floor(float2(rows*rows,rows)*frac(newTime03))/rows)) + randOffset03 + (sin(inputTime*5+(16*seed01.x))*0.0083);\r\nfloat2 staticUV03 = (uvResult03/rows) + randOffset03 + (sin(inputTime*5+(16*seed01.x))*0.0083);\r\n\r\nfloat4 graffiti01 = (Texture2DSample(inputTex, inputTexSampler, animatedUV01).r * Texture2DSampleBias(inputTex, View.MaterialTextureBilinearClampedSampler, staticUV01,-2).a);\r\nfloat4 graffiti02 = (Texture2DSample(inputTex, inputTexSampler, animatedUV02).g * Texture2DSampleBias(inputTex, View.MaterialTextureBilinearClampedSampler, staticUV02,-2).a);\r\nfloat4 graffiti03 = (Texture2DSample(inputTex, inputTexSampler, animatedUV03).b * Texture2DSampleBias(inputTex, View.MaterialTextureBilinearClampedSampler, staticUV03,-2).a);\r\n\r\nresult += ((graffiti01.r * graffitiColor) + (graffiti02.g * graffitiColor) + (graffiti03.b * graffitiColor));\r\nresult = float4(result.r, result.g, result.b, min((result.r+result.g +result.b)*3, 1));\t\t\r\n\t\t\r\n}\r\n\r\nreturn result;\r\n\r\n//inputTexSampler\r\n//View.MaterialTextureBilinearClampedSampler"
   outputType=CMOT_Float4
   showCode=true
   inputs [
    {
     inputName=inputTex
     input {expression=&self:MaterialExpressionTextureObjectParameter_0}
    }
    {
     inputName=rows
     input {expression=&self:MaterialExpressionConstant_16}
    }
    {
     inputName=inputTime
     input {expression=&self:MaterialExpressionTime_7}
    }
    {
     inputName=inputUVs
     input {expression=&self:MaterialExpressionNamedRerouteUsage_11}
    }
    {
     inputName=inputUVs02
     input {expression=&self:MaterialExpressionNamedRerouteUsage_70}
    }
    {
     inputName=graffitiColor01
     input {expression=&self:MaterialExpressionAppendVector_4}
    }
    {
     inputName=graffitiColor02
     input {expression=&self:MaterialExpressionAppendVector_5}
    }
    {
     inputName=graffitiColor03
     input {expression=&self:MaterialExpressionAppendVector_6}
    }
    {
     inputName=graffitiColor04
     input {expression=&self:MaterialExpressionAppendVector_7}
    }
    {
     inputName=graffitiColor05
     input {expression=&self:MaterialExpressionAppendVector_8}
    }
    {
     inputName=uvOffset01Arrow
     input {expression=&self:MaterialExpressionNamedRerouteUsage_71}
    }
    {
     inputName=uvOffset02Arrow
     input {expression=&self:MaterialExpressionNamedRerouteUsage_72}
    }
    {
     inputName=uvOffset03Arrow
     input {expression=&self:MaterialExpressionNamedRerouteUsage_73}
    }
    {
     inputName=uvOffset04Arrow
     input {expression=&self:MaterialExpressionNamedRerouteUsage_74}
    }
    {
     inputName=uvOffset05Arrow
     input {expression=&self:MaterialExpressionNamedRerouteUsage_75}
    }
    {
     inputName=uvOffset01Flower
     input {expression=&self:MaterialExpressionNamedRerouteUsage_76}
    }
    {
     inputName=uvOffset01FlowerBack
     input {expression=&self:MaterialExpressionNamedRerouteUsage_91}
    }
    {
     inputName=uvOffset02Flower
     input {expression=&self:MaterialExpressionNamedRerouteUsage_77}
    }
    {
     inputName=uvOffset03Flower
     input {expression=&self:MaterialExpressionNamedRerouteUsage_78}
    }
    {
     inputName=uvOffset04Flower
     input {expression=&self:MaterialExpressionNamedRerouteUsage_79}
    }
    {
     inputName=uvOffset05Flower
     input {expression=&self:MaterialExpressionNamedRerouteUsage_80}
    }
    {
     inputName=uvOffset01Crown
     input {expression=&self:MaterialExpressionNamedRerouteUsage_81}
    }
    {
     inputName=uvOffset01CrownBack
     input {expression=&self:MaterialExpressionNamedRerouteUsage_89}
    }
    {
     inputName=uvOffset02Crown
     input {expression=&self:MaterialExpressionNamedRerouteUsage_82}
    }
    {
     inputName=uvOffset02CrownBack
     input {expression=&self:MaterialExpressionNamedRerouteUsage_90}
    }
    {
     inputName=uvOffset03Crown
     input {expression=&self:MaterialExpressionNamedRerouteUsage_83}
    }
    {
     inputName=uvOffset04Crown
     input {expression=&self:MaterialExpressionNamedRerouteUsage_84}
    }
    {
     inputName=uvOffset05Crown
     input {expression=&self:MaterialExpressionNamedRerouteUsage_85}
    }
    {
     inputName=loopSpeed
     input {expression=&self:MaterialExpressionScalarParameter_12}
    }
    {
     inputName=revealDuration
     input {expression=&self:MaterialExpressionMultiply_54}
    }
    {
     inputName=seeds
     input {
      expression=&self:MaterialExpressionMaterialFunctionCall_23
      outputIndex=2
     }
    }
    {
     inputName=graffitiFrontMask
     input {expression=&self:MaterialExpressionNamedRerouteUsage_88}
    }
   ]
  }
  n52 = Multiply(A=NamedRerouteUsage(), B=NamedRerouteUsage())
  n53 = NamedRerouteUsage()
  n54 = LinearInterpolate(
   A=@n43.Base Color
   B=ComponentMask(None=@n51) {b=true, g=true, r=true}
   Alpha=Multiply(A=ComponentMask(None=@n51) {a=true}, B=Multiply(A=@n52, B=@n53))
  )
  n55 = Add(A=Min(A=@n43.Metallic) {constB=0.3}, B=NamedRerouteUsage())
  n56 = MaterialFunctionCall() {materialFunction=&MF_CharacterParent_2023_PreFX}
  MaterialFunctionCall(
   Material=SetMaterialAttributes!(
    MaterialAttributes=@n43.MaterialAttributes
    Base Color=StaticSwitchParameter(True=@n54, False=@n43.Base Color) {defaultValue=true, group=SUPERLEVEL, parameterName=UseFlipbookGraffiti}
    Metallic=Saturate(None=@n55)
    Specular=Min(A=@n43.Specular) {constB=0.8}
   ) {attributeSetTypes [#1, #2, #4]}
   M=@n56.M
   Pre Skinned Local Position=@n56.M
   Distance Blend=@n56.M
   Fresnel=@n56.M
  ) {materialFunction=&MF_CharacterParent_2023_PostFX}
  n57 = VectorParameter() {group=RelayStick_Orb, parameterName="OrbRadialCoordScales (UVUV)_hands", defaultValue=<rgba 0.1366 2 0.174 1>}
  n58 = Multiply(
   A=MaterialFunctionCall!(UV Coord=NamedRerouteUsage(), LenghtScale, RadialScale) {materialFunction=/Game/Global/MaterialFunctions/TechArt/MF_PolarCoordinates}.Result
   B=AppendVector(A=@n57.R, B=@n57.R)
  )
  n59 = Saturate(None=Multiply(A=ComponentMask(None=@n58) {r=true}) {constB=12.5})
  n60 = VectorParameter() {group=RelayStick_Orb, parameterName="OrbRadialCoordPanSpeeds (UVUV)_hands", defaultValue=<rgba 0.0333 0.0333 -0.0267 -0.0233>}
  n61 = TextureCoordinate() {uTiling=3, vTiling=3}
  n62 = TextureSample!(
   UVs=Panner(Coordinate=@n58, Time=Time(), Speed=AppendVector(A=@n60.R, B=@n60.R)) {speedX=-0.33, speedY=-0.053}
   Tex
   "DDX(UVs)"=DDX(Value=@n61)
   "DDY(UVs)"=DDY(Value=@n61)
   Apply View MipBias
  ) {
   mipValueMode=TMVM_Derivative
   samplerSource=SSM_Wrap_WorldGroupSettings
   samplerType=SAMPLERTYPE_LinearColor
   texture=/Game/Characters/Player/Male/Medium/Bodies/M_MED_SRDrift/Textures/T_ElectricNoise
  }
  n63 = Power(Base=@n62.R, Exp=ScalarParameter() {defaultValue=2, group=SUPERLEVEL, parameterName=LightningPower})
  Multiply(A=Multiply(A=@n59, B=@n59), B=@n63)
  NamedRerouteDeclaration(None=@n56.Tech Art Mask) {name=FX_Masks, nodeColor=<rgba 0 1 0.2031 1>}
  n64 = NamedRerouteUsage()
  n65 = NamedRerouteUsage()
  n66 = SphereMask(A=@n64, B=@n65, Radius=ScalarParameter() {defaultValue=10, group=SUPERLEVEL, parameterName=ButterflyGlowRadius_Hand_R, sortPriority=101}) {attenuationRadius=30, hardnessPercent=0.3}
  NamedRerouteDeclaration(None=Multiply(A=NamedRerouteDeclaration(None=@n66) {name=GlowMask_Hand_R, nodeColor=<rgba 0 1 0.2266 1>}, B=@n63)) {name=RightHandRays, nodeColor=<rgba 0 0.8125 1 1>}
  NamedRerouteDeclaration(None=WorldPosition().XYZ) {name=AbsoluteWorldPos, nodeColor=<rgba 0 1 0.7188 1>}
  n67 = NamedRerouteUsage()
  n68 = TransformPosition(None=@n67) {transformSourceType=TRANSFORMPOSSOURCE_World, transformType=TRANSFORMPOSSOURCE_View}
  n69 = Multiply(
   A=Divide(B=ComponentMask(None=@n68) {b=true})
   B=ScalarParameter() {defaultValue=10, group=SUPERLEVEL, parameterName=ButterflyScale_Chest, sortPriority=101}
  )
  n70 = NamedRerouteUsage()
  n71 = Add(
   A=MaterialFunctionCall(UVs=NamedRerouteUsage(), Texture Scale=Multiply(A=@n69, B=@n70)) {materialFunction=&ScaleUVsByCenter}.UVs
   B=NamedRerouteUsage()
  )
  n72 = TextureSample!(
   UVs=MaterialFunctionCall(UVs=NamedRerouteDeclaration(None=@n71) {name=ChestUV, nodeColor=<rgba 0 1 0.1797 1>}, Texture Scale=NamedRerouteUsage()) {materialFunction=&ScaleUVsByCenter}.UVs
   Tex
   Apply View MipBias
  ) {
   samplerSource=SSM_Clamp_WorldGroupSettings
   texture=/BRCosmetics/Characters/Player/Female/Medium/Bodies/F_MED_DenimEquip_Stock/Skins/SuperLevel/Textures/T_FN_ClydeSL_Butterfly
  }
  n73 = NamedRerouteUsage()
  n74 = ScalarParameter() {defaultValue=30, group=SUPERLEVEL, parameterName=ButterflyGlowRadius_Chest, sortPriority=101}
  n75 = NamedRerouteDeclaration(None=StaticSwitchParameter(True=SphereMask(A=@n73, B=@n67, Radius=@n74) {attenuationRadius=30, hardnessPercent=0.2}, False=Constant()) {defaultValue=true, group=SUPERLEVEL, parameterName=UseButterfly_Chest, sortPriority=101}) {name=GlowMask_Chest, nodeColor=<rgba 1 0.8672 0 1>}
  NamedRerouteDeclaration(None=Multiply(A=@n72.A, B=@n75)) {name=Chest_Alpha, nodeColor=<rgba 0 1 0.6953 1>}
  NamedRerouteDeclaration(None=StaticSwitchParameter(True=@n72.RGB, False=Constant()) {defaultValue=true, group=SUPERLEVEL, parameterName=UseButterfly_Chest, sortPriority=101}) {name=Chest_RGB, nodeColor=<rgba 1 0.375 0 1>}
  n76 = NamedRerouteUsage()
  n77 = TransformPosition(None=@n76) {transformSourceType=TRANSFORMPOSSOURCE_World, transformType=TRANSFORMPOSSOURCE_View}
  n78 = Multiply(
   A=Divide(B=ComponentMask(None=@n77) {b=true})
   B=ScalarParameter() {defaultValue=5, group=SUPERLEVEL, parameterName=ButterflyScale_Hand_L, sortPriority=101}
  )
  n79 = NamedRerouteUsage()
  n80 = Add(
   A=MaterialFunctionCall(UVs=NamedRerouteUsage(), Texture Scale=Multiply(A=@n78, B=@n79)) {materialFunction=&ScaleUVsByCenter}.UVs
   B=NamedRerouteUsage()
  )
  n81 = TextureSample(UVs=MaterialFunctionCall(UVs=NamedRerouteDeclaration(None=@n80) {name=L_Hand_UV, nodeColor=<rgba 0 1 0.4609 1>}, Texture Scale=NamedRerouteUsage()) {materialFunction=&ScaleUVsByCenter}.UVs) {
   samplerSource=SSM_Clamp_WorldGroupSettings
   texture=&T_FN_ClydeSL_Butterfly
  }
  n82 = NamedRerouteUsage()
  n83 = SphereMask(
   A=@n82
   B=@n76
   Radius=ScalarParameter() {defaultValue=18, group=SUPERLEVEL, parameterName=ButterflyVisibilityRadius_Hand_L, sortPriority=101}
  ) {attenuationRadius=30, hardnessPercent=0.3}
  n84 = Constant()
  NamedRerouteDeclaration(None=Multiply(
    A=@n81.A
    B=StaticSwitchParameter(True=@n83, False=@n84) {defaultValue=true, group=SUPERLEVEL, parameterName=UseButterfly_Hand_L, sortPriority=100}
   )) {name=L_Hand_Alpha, nodeColor=<rgba 0.0547 0 1 1>}
  n85 = Time()
  n86 = Sine(None=Multiply(A=@n85) {constB=0.4})
  NamedRerouteDeclaration(None=Multiply(A=Multiply(A=@n86) {constB=0.01}, B=Constant2Vector() {g=1})) {name=ButterflyWiggleUV, nodeColor=<rgba 0.6406 1 0 1>}
  NamedRerouteDeclaration(None=StaticSwitchParameter(True=@n81.RGB, False=Constant()) {defaultValue=true, group=SUPERLEVEL, parameterName=UseButterfly_Hand_L, sortPriority=100}) {name=L_Hand_RGB, nodeColor=<rgba 0.2656 1 0 1>}
  n87 = TransformPosition(None=@n65) {transformSourceType=TRANSFORMPOSSOURCE_World, transformType=TRANSFORMPOSSOURCE_View}
  n88 = Multiply(
   A=Divide(B=ComponentMask(None=@n87) {b=true})
   B=ScalarParameter() {defaultValue=5, group=SUPERLEVEL, parameterName=ButterflyScale_Hand_R, sortPriority=101}
  )
  n89 = NamedRerouteUsage()
  n90 = Add(
   A=MaterialFunctionCall(UVs=NamedRerouteUsage(), Texture Scale=Multiply(A=@n88, B=@n89)) {materialFunction=&ScaleUVsByCenter}.UVs
   B=NamedRerouteUsage()
  )
  n91 = TextureSample(UVs=MaterialFunctionCall(UVs=NamedRerouteDeclaration(None=@n90) {name=R_Hand_UV, nodeColor=<rgba 1 0 0.0937 1>}, Texture Scale=NamedRerouteUsage()) {materialFunction=&ScaleUVsByCenter}.UVs) {
   samplerSource=SSM_Clamp_WorldGroupSettings
   texture=&T_FN_ClydeSL_Butterfly
  }
  n92 = ScalarParameter() {defaultValue=18, group=SUPERLEVEL, parameterName=ButterflyVisibilityRadius_Hand_R, sortPriority=101}
  NamedRerouteDeclaration(None=Multiply(A=@n91.A, B=SphereMask(A=@n64, B=@n65, Radius=@n92) {attenuationRadius=30, hardnessPercent=0.3})) {name=R_Hand_Alpha, nodeColor=<rgba 0.6875 0 1 1>}
  NamedRerouteDeclaration(None=@n91.RGB) {name=R_Hand_RGB, nodeColor=<rgba 0 1 0.3672 1>}
  n93 = NamedRerouteUsage()
  n94 = NamedRerouteUsage()
  n95 = Multiply(
   A=Add(A=NamedRerouteUsage(), B=Multiply(A=@n93, B=@n94))
   B=ScalarParameter() {defaultValue=0.3, group=SUPERLEVEL, parameterName=Butterfly_HandsBrightnessMultiplier}
  ) {constB=0.4}
  n96 = Multiply(A=NamedRerouteUsage()) {constB=1000}
  n97 = Multiply(
   A=ScalarParameter() {defaultValue=1, group=SUPERLEVEL, parameterName=Butterfly_ChestBrightnessMultiplier}
   B=Multiply(A=NamedRerouteUsage(), B=Saturate(None=@n96))
  )
  NamedRerouteDeclaration(None=StaticSwitchParameter(True=Add(A=@n95, B=@n97), False=Constant()) {group=SUPERLEVEL, parameterName=UseButterfly, sortPriority=100}) {name=Butterfly_RGB, nodeColor=<rgba 1 0.0469 0 1>}
  n98 = Add(
   A=Add(A=NamedRerouteUsage(), B=NamedRerouteUsage())
   B=StaticSwitchParameter(True=Constant(), False=NamedRerouteUsage()) {group=SUPERLEVEL, parameterName="MakeChestButterflyInvisible(StageCue)", sortPriority=100}
  )
  NamedRerouteDeclaration(None=StaticSwitchParameter(True=Saturate(None=@n98), False=Constant()) {group=SUPERLEVEL, parameterName=UseButterfly, sortPriority=100}) {name=Butterfly_Alpha, nodeColor=<rgba 0 1 0.0391 1>}
  n99 = TextureSampleParameter2D!(UVs, Level, Apply View MipBias) {
   group=SUPERLEVEL
   mipValueMode=TMVM_MipLevel
   parameterName=Superlevel_Masks
   samplerSource=SSM_Wrap_WorldGroupSettings
   samplerType=SAMPLERTYPE_Masks
   sortPriority=0
   texture=/BRCosmetics/Characters/Player/Female/Medium/Bodies/F_MED_DenimEquip_Stock/Textures/T_DenimEquip_Stock_Superlevel_Body_Mask
  }
  NamedRerouteDeclaration(None=@n99.R) {name=PaintMask, nodeColor=<rgba 1 0 0.0589 1>}
  NamedRerouteDeclaration(None=@n99.G) {name=AccentMask, nodeColor=<rgba 0 1 0.0638 1>}
  NamedRerouteDeclaration(None=@n99.B) {name=ButterflyMask, nodeColor=<rgba 0 0.1106 1 1>}
  NamedRerouteDeclaration(None=@n56.Result) {name=OriginalMaterial, nodeColor=<rgba 0.2422 1 0 1>}
  n100 = TextureSampleParameter2D!(UVs, Apply View MipBias) {
   group=00_Texture Inputs
   parameterName=Mask
   samplerSource=SSM_Wrap_WorldGroupSettings
   samplerType=SAMPLERTYPE_Masks
   sortPriority=3
   texture=&T_BlackMask
  }
  NamedRerouteDeclaration(None=@n100.B) {name=SkinMask, nodeColor=<rgba 0.1484 0 1 1>}
  n101 = Multiply(A=Sine(None=Multiply(A=@n85) {constB=0.23})) {constB=0.5}
  NamedRerouteDeclaration(None=LinearInterpolate(Alpha=Add(B=@n101) {constA=0.5}) {constA=0.85}) {name=ButterflyAnimatedScale, nodeColor=<rgba 1 0 0.9609 1>}
  n102 = NamedRerouteUsage()
  n103 = Distance(A=NamedRerouteUsage(), B=NamedRerouteUsage())
  n104 = Max(A=Saturate(None=Divide(A=@n103) {constB=60})) {constB=0.2}
  n105 = Multiply(A=@n104, B=NamedRerouteUsage())
  n106 = NamedRerouteUsage()
  n107 = Saturate(None=Add(A=Multiply(A=@n102, B=@n104), B=Add(A=@n105, B=@n106)))
  n108 = Multiply(A=ScalarParameter() {group=SUPERLEVEL, parameterName=ButterflyGlow_ShiftHue, sortPriority=101}, B=@n107)
  n109 = Constant()
  NamedRerouteDeclaration(None=StaticSwitchParameter(True=Constant() {r=1}, False=StaticSwitchParameter(True=@n108, False=@n109) {group=SUPERLEVEL, parameterName=UseButterfly, sortPriority=100}) {group=SUPERLEVEL, parameterName=UseButterfly_AdditionalRaysOnly, sortPriority=101}) {name=ButterflyShiftHue, nodeColor=<rgba 1 0 0.75 1>}
  n110 = ScalarParameter() {group=SUPERLEVEL, parameterName=ButterflyGlow_Brightness, sortPriority=101}
  n111 = Add(A=Multiply(A=ScalarParameter() {group=SUPERLEVEL, parameterName=ButterflyGlow_Brightness, sortPriority=101}, B=@n107))
  n112 = Constant() {r=1}
  NamedRerouteDeclaration(None=StaticSwitchParameter(True=Add(A=@n110), False=StaticSwitchParameter(True=@n111, False=@n112) {group=SUPERLEVEL, parameterName=UseButterfly, sortPriority=100}) {group=SUPERLEVEL, parameterName=UseButterfly_AdditionalRaysOnly, sortPriority=101}) {name=ButterflyBrightness, nodeColor=<rgba 0 1 0.9297 1>}
  n113 = Constant2Vector() {g=0.9, r=1.1}
  n114 = Constant2Vector() {g=1.1, r=0.9}
  n115 = Multiply(A=Time()) {constB=0.9}
  n116 = Multiply(A=Sine(None=Add(A=@n115) {constB=0.6})) {constB=0.5}
  NamedRerouteDeclaration(None=LinearInterpolate(A=@n113, B=@n114, Alpha=Add(A=@n116) {constB=0.5})) {name=ButterflyFlappingWings_L, nodeColor=<rgba 0 1 0.6484 1>}
  n117 = Sine(None=Add(A=@n115) {constB=0.3})
  n118 = LinearInterpolate(A=@n113, B=@n114, Alpha=Add(A=Multiply(A=@n117) {constB=0.5}) {constB=0.5})
  NamedRerouteDeclaration(None=@n118) {name=ButterflyFlappingWings_R, nodeColor=<rgba 0 1 0.6484 1>}
  n119 = LocalPosition() {localOrigin=InstancePreSkinning}
  n120 = NamedRerouteDeclaration(None=ShaderStageSwitch(PixelShader=VertexInterpolator(VS=@n119.XYZ).PS, VertexShader=@n119.XYZ)) {name=PreskinnedPixel, nodeColor=<rgba 0 0.5547 1 1>}
  n121 = Add(A=Constant3Vector() {constant=<rgba -30 -15 -75 0>}, B=@n120)
  NamedRerouteDeclaration(None=Divide(A=ComponentMask(None=@n121) {b=true, r=true}) {constB=15}) {name=PlanarProjectUV_Front, nodeColor=<rgba 1 0 0.7031 1>}
  n122 = TextureCoordinate() {uTiling=3, vTiling=3}
  n123 = ScalarParameter() {defaultValue=1.5, group=SUPERLEVEL, parameterName=AdditionalRaysScale, sortPriority=20}
  n124 = VectorParameter() {group=SUPERLEVEL, parameterName=AdditionalRays_PanSpeed, sortPriority=20, defaultValue=<rgba 0.05 0.1 0 0>}
  n125 = TextureCoordinate() {uTiling=3, vTiling=3}
  n126 = TextureSample!(
   UVs=Panner(Coordinate=Multiply(A=@n122, B=@n123), Speed=ComponentMask(None=@n124.RGB) {g=true, r=true}) {speedX=0.05, speedY=-0.1}
   Tex
   "DDX(UVs)"=DDX(Value=@n125)
   "DDY(UVs)"=DDY(Value=@n125)
   Apply View MipBias
  ) {
   mipValueMode=TMVM_Derivative
   samplerSource=SSM_Wrap_WorldGroupSettings
   samplerType=SAMPLERTYPE_LinearColor
   texture=&T_ElectricNoise
  }
  n127 = ScalarParameter() {defaultValue=2, group=SUPERLEVEL, parameterName=LightningPower}
  n128 = Multiply(
   A=Power(Base=@n126.RGB, Exp=Multiply(A=@n127) {constB=1.3})
   B=ScalarParameter() {defaultValue=1, group=SUPERLEVEL, parameterName=AdditionalRays_Brightness, sortPriority=20}
  )
  n129 = Subtract(
   A=ComponentMask(None=NamedRerouteUsage()) {b=true}
   B=ScalarParameter() {group=SUPERLEVEL, parameterName=AdditionalRays_OffsetMask, sortPriority=20}
  )
  n130 = Add(A=Add(A=NamedRerouteUsage(), B=NamedRerouteUsage()), B=NamedRerouteUsage())
  n131 = Multiply(A=OneMinus(None=Saturate(None=@n129)), B=OneMinus(None=Saturate(None=@n130)))
  NamedRerouteDeclaration(None=StaticSwitchParameter(True=Multiply(A=@n128, B=@n131), False=Constant()) {group=SUPERLEVEL, parameterName=UseButterfly_AdditionalRays, sortPriority=101}) {name=AdditionalRays, nodeColor=<rgba 0.4062 0 1 1>}
  n132 = NamedRerouteUsage()
  n133 = TextureSample!(UVs, Tex, Apply View MipBias) {
   constCoordinate=1
   samplerSource=SSM_Wrap_WorldGroupSettings
   samplerType=SAMPLERTYPE_Masks
   texture=/BRCosmetics/Characters/Player/Male/Large/Bodies/M_LRG_GoldCat_Claw/Textures/goldCatClaw_tattoo_texTransfer
  }
  n134 = Multiply(A=@n133.G, B=ChannelMaskParameter(None=NamedRerouteUsage()) {parameterName=GC_TattooMask}.R)
  n135 = Constant()
  n136 = LinearInterpolate(
   A=GetMaterialAttributes(None=NamedRerouteUsage()) {attributeGetTypes [#1]}.Base Color
   B=VectorParameter() {group=GoldCatClaw, parameterName=GC_TattooColor, defaultValue=<rgba 0.0742 0.0648 0.0545 1>}.RGB
   Alpha=NamedRerouteDeclaration(None=StaticSwitchParameter(True=@n134, False=@n135) {group=GoldCatClaw, parameterName=UseGoldCatTattoos}) {name=TattooMask, nodeColor=<rgba 1 0 0.5391 1>}
  )
  NamedRerouteDeclaration(None=StaticSwitchParameter(
    True=SetMaterialAttributes!(MaterialAttributes=@n132, Base Color=@n136) {attributeSetTypes [#1]}
    False=NamedRerouteUsage()
   ) {group=GoldCatClaw, parameterName=UseGoldCatTattoos}) {name=BaseMaterial, nodeColor=<rgba 1 0 0.1172 1>}
  n137 = MaterialFunctionCall() {materialFunction=&CameraDirectionVector}
  n138 = Multiply(A=Constant3Vector() {constant=<rgba 1 1 0 0>}, B=@n137.Result)
  n139 = Add(A=Multiply(B=Normalize(VectorInput=@n138)) {constA=0.25}, B=@n137.Result)
  n140 = VectorParameter() {group=SUPERLEVEL, parameterName=LocalChestPosition, sortPriority=1000, defaultValue=<rgba 0 0 132 0>}
  n141 = Add(A=@n140.RGB, B=VectorParameter() {group=SUPERLEVEL, parameterName=ButterflyPreviewOffset_Chest, sortPriority=800}.RGB)
  n142 = Multiply(
   A=ScalarParameter() {group=SUPERLEVEL, parameterName=PositionOffset_Chest, sortPriority=150}
   B=Normalize(VectorInput=VectorParameter() {group=SUPERLEVEL, parameterName=BP_OffsetVector_Chest, sortPriority=1000}.RGB)
  )
  n143 = StaticSwitchParameter(
   True=VectorParameter() {group=SUPERLEVEL, parameterName=BP_ChestPositionMustard, sortPriority=1000}.RGB
   False=VectorParameter() {group=SUPERLEVEL, parameterName=BP_ChestPosition, sortPriority=1000}.RGB
  ) {group=SUPERLEVEL, parameterName=UseStageCueMustard_CCPM}
  n144 = NamedRerouteDeclaration(None=LinearInterpolate(
    A=TransformPosition(None=@n141) {transformType=TRANSFORMPOSSOURCE_World}
    B=Add(A=@n142, B=@n143)
    Alpha=ScalarParameter() {group=SUPERLEVEL, parameterName=BP_UseCCPM, sortPriority=0}
   )) {name=ChestPosition, nodeColor=<rgba 0 1 0.2734 1>}
  n145 = Constant() {r=45}
  n146 = Constant() {r=90}
  n147 = MaterialFunctionCall() {materialFunction=/Engine/Functions/Engine_MaterialFunctions02/Utility/FOV}
  n148 = Multiply(
   A=LinearInterpolate(A=Constant() {r=64}, B=Constant() {r=128}, Alpha=InverseLinearInterpolate(A=@n145, B=@n146, Value=@n147.FOV angle in degrees))
   B=Distance(A=@n144, B=CameraPositionWS())
  )
  NamedRerouteDeclaration(None=MaterialFunctionCall(Plane Normal Axis=Normalize(VectorInput=@n139), Plane Center=Add(A=@n137.Result, B=@n144), UV Scale=Multiply(A=@n148) {constB=0.0146}) {materialFunction=&VirtualPlaneCoordinates}.UVs) {name=ChestAllignedScreenPosition, nodeColor=<rgba 0.125 1 0 1>}
  n149 = Add(A=Add(A=@n15.R, B=@n15.R), B=@n15.B)
  NamedRerouteDeclaration(None=Saturate(None=Multiply(A=@n149) {constB=2})) {name=SplotchesMask, nodeColor=<rgba 1 0.1172 0 1>}
  n150 = MaterialFunctionCall() {materialFunction=&CameraDirectionVector}
  n151 = Multiply(A=Constant3Vector() {constant=<rgba 1 1 0 0>}, B=@n150.Result)
  n152 = Add(A=Multiply(B=Normalize(VectorInput=@n151)) {constA=0.25}, B=@n150.Result)
  n153 = VectorParameter() {group=SUPERLEVEL, parameterName=LocalHandsPosition, sortPriority=1000, defaultValue=<rgba 35 5 88 0>}
  n154 = Add(
   A=Multiply(A=@n153.RGB, B=Constant3Vector() {constant=<rgba -1 1 1 0>})
   B=VectorParameter() {group=SUPERLEVEL, parameterName=ButterflyPreviewOffset_Hand_R, sortPriority=800}.RGB
  )
  n155 = Multiply(
   A=ScalarParameter() {group=SUPERLEVEL, parameterName=PositionOffset_R, sortPriority=150}
   B=Normalize(VectorInput=VectorParameter() {group=SUPERLEVEL, parameterName=BP_OffsetVector_R, sortPriority=1000}.RGB)
  )
  n156 = VectorParameter() {group=SUPERLEVEL, parameterName=BP_HandPosition_R, sortPriority=1000}
  n157 = NamedRerouteDeclaration(None=LinearInterpolate(
    A=TransformPosition(None=@n154) {transformType=TRANSFORMPOSSOURCE_World}
    B=Add(A=@n155, B=@n156.RGB)
    Alpha=ScalarParameter() {group=SUPERLEVEL, parameterName=BP_UseCCPM, sortPriority=0}
   )) {name=Hand_R_Position, nodeColor=<rgba 0 1 0.8594 1>}
  n158 = Constant() {r=45}
  n159 = Constant() {r=90}
  n160 = MaterialFunctionCall() {materialFunction=&FOV}
  n161 = Multiply(
   A=LinearInterpolate(A=Constant() {r=64}, B=Constant() {r=128}, Alpha=InverseLinearInterpolate(A=@n158, B=@n159, Value=@n160.FOV angle in degrees))
   B=Distance(A=@n157, B=CameraPositionWS())
  )
  NamedRerouteDeclaration(None=MaterialFunctionCall(Plane Normal Axis=Normalize(VectorInput=@n152), Plane Center=Add(A=@n150.Result, B=@n157), UV Scale=Multiply(A=@n161) {constB=0.0146}) {materialFunction=&VirtualPlaneCoordinates}.UVs) {name=Hand_R_AllignedScreenPosition, nodeColor=<rgba 0.125 1 0 1>}
  n162 = MaterialFunctionCall() {materialFunction=&CameraDirectionVector}
  n163 = Multiply(A=Constant3Vector() {constant=<rgba 1 1 0 0>}, B=@n162.Result)
  n164 = Add(A=Multiply(B=Normalize(VectorInput=@n163)) {constA=0.25}, B=@n162.Result)
  n165 = VectorParameter() {group=SUPERLEVEL, parameterName=LocalHandsPosition, sortPriority=1000, defaultValue=<rgba 35 5 88 0>}
  n166 = Add(A=@n165.RGB, B=VectorParameter() {group=SUPERLEVEL, parameterName=ButterflyPreviewOffset_Hand_L, sortPriority=800}.RGB)
  n167 = Multiply(
   A=ScalarParameter() {group=SUPERLEVEL, parameterName=PositionOffset_L, sortPriority=150}
   B=Normalize(VectorInput=VectorParameter() {group=SUPERLEVEL, parameterName=BP_OffsetVector_L, sortPriority=1000}.RGB)
  )
  n168 = VectorParameter() {group=SUPERLEVEL, parameterName=BP_HandPosition_L, sortPriority=1000}
  n169 = NamedRerouteDeclaration(None=LinearInterpolate(
    A=TransformPosition(None=@n166) {transformType=TRANSFORMPOSSOURCE_World}
    B=Add(A=@n167, B=@n168.RGB)
    Alpha=ScalarParameter() {group=SUPERLEVEL, parameterName=BP_UseCCPM, sortPriority=0}
   )) {name=Hand_L_Position, nodeColor=<rgba 0 1 0.8594 1>}
  n170 = Constant() {r=45}
  n171 = Constant() {r=90}
  n172 = MaterialFunctionCall() {materialFunction=&FOV}
  n173 = Multiply(
   A=LinearInterpolate(A=Constant() {r=64}, B=Constant() {r=128}, Alpha=InverseLinearInterpolate(A=@n170, B=@n171, Value=@n172.FOV angle in degrees))
   B=Distance(A=@n169, B=CameraPositionWS())
  )
  NamedRerouteDeclaration(None=MaterialFunctionCall(Plane Normal Axis=Normalize(VectorInput=@n164), Plane Center=Add(A=@n162.Result, B=@n169), UV Scale=Multiply(A=@n173) {constB=0.0146}) {materialFunction=&VirtualPlaneCoordinates}.UVs) {name=Hand_L_AllignedScreenPosition, nodeColor=<rgba 0.125 1 0 1>}
  NamedRerouteDeclaration(None=Divide(A=ComponentMask(None=@n121) {b=true, g=true}) {constB=15}) {name=PlanarProjectUV_Side, nodeColor=<rgba 1 0 0.7031 1>}
  n174 = VectorParameter() {group=SUPERLEVEL, parameterName=ARROW_GraffitiOffset/Rot/Scale01, sortPriority=41, defaultValue=<rgba 1 1 1 1>}
  NamedRerouteDeclaration(None=AppendVector(A=@n174.RGB, B=@n174.RGB)) {name=OffsetUVs_Arrow01, nodeColor=<rgba 0 1 0.2031 1>}
  n175 = VectorParameter() {group=SUPERLEVEL, parameterName=ARROW_GraffitiOffset/Rot/Scale02, sortPriority=43, defaultValue=<rgba 1 1 1 1>}
  NamedRerouteDeclaration(None=AppendVector(A=@n175.RGB, B=@n175.RGB)) {name=OffsetUVs_Arrow02, nodeColor=<rgba 0 1 0.3543 1>}
  n176 = VectorParameter() {group=SUPERLEVEL, parameterName=ARROW_GraffitiOffset/Rot/Scale03, sortPriority=45, defaultValue=<rgba 1 1 1 1>}
  NamedRerouteDeclaration(None=AppendVector(A=@n176.RGB, B=@n176.RGB)) {name=OffsetUVs_Arrow03, nodeColor=<rgba 0 1 0.3543 1>}
  n177 = VectorParameter() {group=SUPERLEVEL, parameterName=ARROW_GraffitiOffset/Rot/Scale04, sortPriority=47, defaultValue=<rgba 1 1 1 1>}
  NamedRerouteDeclaration(None=AppendVector(A=@n177.RGB, B=@n177.RGB)) {name=OffsetUVs_Arrow04, nodeColor=<rgba 0 1 0.3543 1>}
  n178 = Add(A=Multiply(A=Sine(None=@n115)) {constB=0.5}) {constB=0.5}
  NamedRerouteDeclaration(None=StaticSwitchParameter(True=@n118, False=LinearInterpolate(A=@n113, B=@n114, Alpha=@n178)) {group=SUPERLEVEL, parameterName="MatchChestAndHandsFlapping(StageCue)", sortPriority=100}) {name=ButterflyFlappingWings_Chest, nodeColor=<rgba 0 1 0.6484 1>}
  n179 = VectorParameter() {group=SUPERLEVEL, parameterName=ARROW_GraffitiOffset/Rot/Scale05, sortPriority=49, defaultValue=<rgba 1 1 1 1>}
  NamedRerouteDeclaration(None=AppendVector(A=@n179.RGB, B=@n179.RGB)) {name=OffsetUVs_Arrow05, nodeColor=<rgba 0 1 0.3543 1>}
  n180 = VectorParameter() {group=SUPERLEVEL, parameterName=FLOWER_GraffitiOffset/Rot/Scale01, sortPriority=41, defaultValue=<rgba 1 1 1 1>}
  n181 = AppendVector(A=@n180.RGB, B=@n180.RGB)
  NamedRerouteDeclaration(None=@n181) {name=OffsetUVs_Flower01, nodeColor=<rgba 0 0.9003 1 1>}
  n182 = VectorParameter() {group=SUPERLEVEL, parameterName=FLOWER_GraffitiOffset/Rot/Scale02, sortPriority=43, defaultValue=<rgba 1 1 1 1>}
  NamedRerouteDeclaration(None=AppendVector(A=@n182.RGB, B=@n182.RGB)) {name=OffsetUVs_Flower02, nodeColor=<rgba 0 0.9439 1 1>}
  n183 = VectorParameter() {group=SUPERLEVEL, parameterName=FLOWER_GraffitiOffset/Rot/Scale03, sortPriority=45, defaultValue=<rgba 1 1 1 1>}
  NamedRerouteDeclaration(None=AppendVector(A=@n183.RGB, B=@n183.RGB)) {name=OffsetUVs_Flower03, nodeColor=<rgba 0 1 0.9193 1>}
  n184 = VectorParameter() {group=SUPERLEVEL, parameterName=FLOWER_GraffitiOffset/Rot/Scale04, sortPriority=47, defaultValue=<rgba 1 1 1 1>}
  NamedRerouteDeclaration(None=AppendVector(A=@n184.RGB, B=@n184.RGB)) {name=OffsetUVs_Flower04, nodeColor=<rgba 0 1 0.9511 1>}
  n185 = VectorParameter() {group=SUPERLEVEL, parameterName=FLOWER_GraffitiOffset/Rot/Scale05, sortPriority=49, defaultValue=<rgba 1 1 1 1>}
  NamedRerouteDeclaration(None=AppendVector(A=@n185.RGB, B=@n185.RGB)) {name=OffsetUVs_Flower05, nodeColor=<rgba 0 1 0.9355 1>}
  n186 = VectorParameter() {group=SUPERLEVEL, parameterName=CROWN_GraffitiOffset/Rot/Scale01, sortPriority=41, defaultValue=<rgba 1 1 1 1>}
  n187 = AppendVector(A=@n186.RGB, B=@n186.RGB)
  NamedRerouteDeclaration(None=@n187) {name=OffsetUVs_Crown01, nodeColor=<rgba 0.39 0 1 1>}
  n188 = VectorParameter() {group=SUPERLEVEL, parameterName=CROWN_GraffitiOffset/Rot/Scale02, sortPriority=43, defaultValue=<rgba 1 1 1 1>}
  n189 = AppendVector(A=@n188.RGB, B=@n188.RGB)
  NamedRerouteDeclaration(None=@n189) {name=OffsetUVs_Crown02, nodeColor=<rgba 0.4253 0 1 1>}
  n190 = VectorParameter() {group=SUPERLEVEL, parameterName=CROWN_GraffitiOffset/Rot/Scale03, sortPriority=45, defaultValue=<rgba 1 1 1 1>}
  NamedRerouteDeclaration(None=AppendVector(A=@n190.RGB, B=@n190.RGB)) {name=OffsetUVs_Crown03, nodeColor=<rgba 0.2971 0 1 1>}
  n191 = VectorParameter() {group=SUPERLEVEL, parameterName=CROWN_GraffitiOffset/Rot/Scale04, sortPriority=47, defaultValue=<rgba 1 1 1 1>}
  NamedRerouteDeclaration(None=AppendVector(A=@n191.RGB, B=@n191.RGB)) {name=OffsetUVs_Crown04, nodeColor=<rgba 0.4213 0 1 1>}
  n192 = VectorParameter() {group=SUPERLEVEL, parameterName=CROWN_GraffitiOffset/Rot/Scale05, sortPriority=49, defaultValue=<rgba 1 1 1 1>}
  NamedRerouteDeclaration(None=AppendVector(A=@n192.RGB, B=@n192.RGB)) {name=OffsetUVs_Crown05, nodeColor=<rgba 0.4079 0 1 1>}
  n193 = SphereMask(
   A=NamedRerouteUsage()
   B=Constant2Vector() {g=0.5, r=0.5}
   Radius=ScalarParameter() {defaultValue=30, group=SUPERLEVEL, parameterName=ButterflyOverTheClothes_CenterGlowRadius, sortPriority=101}
   Hardness=ScalarParameter() {defaultValue=0.2, group=SUPERLEVEL, parameterName=ButterflyOverTheClothes_CenterGlowHardness, sortPriority=101}
  ) {attenuationRadius=30, hardnessPercent=0.2}
  NamedRerouteDeclaration(None=Multiply(A=@n193, B=@n193)) {name=OverTheClothes_SphereMask, nodeColor=<rgba 1 0 0.0703 1>}
  n194 = ComponentMask(None=@n120) {g=true}
  NamedRerouteDeclaration(None=Saturate(None=Multiply(A=@n194) {constB=100})) {name=Graffiti_FrontMask, nodeColor=<rgba 1 0 0.3984 1>}
  n195 = VectorParameter() {group=SUPERLEVEL, parameterName=CROWN_Back_GraffitiOffset/Rot/Scale01, sortPriority=41, defaultValue=<rgba 1 1 1 1>}
  NamedRerouteDeclaration(None=StaticSwitchParameter(True=AppendVector(A=@n195.RGB, B=@n195.RGB), False=@n187) {group=SUPERLEVEL, parameterName=Graffiti_OffsetBackCrown_01}) {name=OffsetUVs_Crown01_Back, nodeColor=<rgba 0.39 0 1 1>}
  n196 = VectorParameter() {group=SUPERLEVEL, parameterName=CROWN_Back_GraffitiOffset/Rot/Scale02, sortPriority=43, defaultValue=<rgba 1 1 1 1>}
  NamedRerouteDeclaration(None=StaticSwitchParameter(True=AppendVector(A=@n196.RGB, B=@n196.RGB), False=@n189) {group=SUPERLEVEL, parameterName=Graffiti_OffsetBackCrown_02}) {name=OffsetUVs_Crown02_Back, nodeColor=<rgba 0.4253 0 1 1>}
  n197 = VectorParameter() {group=SUPERLEVEL, parameterName=FLOWER_Back_GraffitiOffset/Rot/Scale01, sortPriority=41, defaultValue=<rgba 1 1 1 1>}
  NamedRerouteDeclaration(None=StaticSwitchParameter(True=AppendVector(A=@n197.RGB, B=@n197.RGB), False=@n181) {group=SUPERLEVEL, parameterName=Graffiti_OffsetBackFlower_01}) {name=OffsetUVs_Flower01_Back, nodeColor=<rgba 0 0.9003 1 1>}
  n198 = Constant()
  n199 = Add(A=NamedRerouteUsage(), B=Add(A=NamedRerouteUsage(), B=NamedRerouteUsage()))
  n200 = Saturate(None=Add(
    A=StaticSwitchParameter(True=@n198, False=@n199) {group=SUPERLEVEL, parameterName=UseButterfly_AdditionalRaysOnly, sortPriority=101}
    B=NamedRerouteUsage()
   ))
  n201 = VectorParameter() {group=SUPERLEVEL, parameterName=LightningColors, defaultValue=<rgba 0.5185 0.6882 1 1>}
  NamedRerouteDeclaration(None=Multiply(A=Multiply(A=@n200, B=@n201.RGB), B=ScalarParameter() {defaultValue=2, group=SUPERLEVEL, parameterName=LightningIntensity}) {constB=10}) {name=RiftRays, nodeColor=<rgba 1 0 0.5625 1>}
  n202 = VectorParameter() {group=RelayStick_Orb, parameterName="OrbRadialCoordScales (UVUV)", defaultValue=<rgba 0.1366 2 0.174 1>}
  n203 = Multiply(
   A=MaterialFunctionCall(UV Coord=NamedRerouteUsage()) {materialFunction=&MF_PolarCoordinates}.Result
   B=AppendVector(A=@n202.B, B=@n202.B)
  )
  n204 = Saturate(None=Multiply(A=ComponentMask(None=@n203) {r=true}) {constB=10})
  n205 = VectorParameter() {group=RelayStick_Orb, parameterName="OrbRadialCoordPanSpeeds (UVUV)", defaultValue=<rgba 0.0333 0.0333 -0.0267 -0.0233>}
  n206 = TextureCoordinate() {uTiling=3, vTiling=3}
  n207 = TextureSample!(
   UVs=Panner(Coordinate=@n203, Time=Time(), Speed=AppendVector(A=@n205.B, B=@n205.B)) {speedX=-0.33, speedY=-0.053}
   Tex
   "DDX(UVs)"=DDX(Value=@n206)
   "DDY(UVs)"=DDY(Value=@n206)
   Apply View MipBias
  ) {
   mipValueMode=TMVM_Derivative
   samplerSource=SSM_Wrap_WorldGroupSettings
   samplerType=SAMPLERTYPE_LinearColor
   texture=&T_ElectricNoise
  }
  n208 = Power(Base=@n207.R, Exp=ScalarParameter() {defaultValue=2, group=SUPERLEVEL, parameterName=LightningPower})
  n209 = Multiply(A=Multiply(A=@n204, B=@n204), B=@n208)
  NamedRerouteDeclaration(None=Multiply(A=@n75, B=StaticSwitchParameter(True=@n209, False=@n208) {group=SUPERLEVEL, parameterName=ReduceRiftSquashing_Chest})) {name=ChestRays, nodeColor=<rgba 0.1719 1 0 1>}
  n210 = StaticSwitchParameter(
   True=SphereMask(A=@n82, B=@n76, Radius=ScalarParameter() {defaultValue=10, group=SUPERLEVEL, parameterName=ButterflyGlowRadius_Hand_L, sortPriority=101}) {attenuationRadius=30, hardnessPercent=0.3}
   False=@n84
  ) {defaultValue=true, group=SUPERLEVEL, parameterName=UseButterfly_Hand_L, sortPriority=100}
  n211 = MaterialFunctionCall(UV Coord=NamedRerouteUsage()) {materialFunction=&MF_PolarCoordinates}
  n212 = VectorParameter() {group=RelayStick_Orb, parameterName="OrbRadialCoordScales (UVUV)_hands", defaultValue=<rgba 0.1366 2 0.174 1>}
  n213 = AppendVector(A=@n212.B, B=@n212.B)
  n214 = VectorParameter() {group=RelayStick_Orb, parameterName="OrbRadialCoordPanSpeeds (UVUV)_hands", defaultValue=<rgba 0.0333 0.0333 -0.0267 -0.0233>}
  n215 = TextureCoordinate() {uTiling=3, vTiling=3}
  n216 = TextureSample(
   UVs=Panner(Coordinate=Multiply(A=@n211.Result, B=@n213), Time=Time(), Speed=AppendVector(A=@n214.B, B=@n214.B)) {speedX=-0.33, speedY=-0.053}
   "DDX(UVs)"=DDX(Value=@n215)
   "DDY(UVs)"=DDY(Value=@n215)
  ) {
   mipValueMode=TMVM_Derivative
   samplerSource=SSM_Wrap_WorldGroupSettings
   samplerType=SAMPLERTYPE_LinearColor
   texture=&T_ElectricNoise
  }
  n217 = ScalarParameter() {defaultValue=2, group=SUPERLEVEL, parameterName=LightningPower}
  NamedRerouteDeclaration(None=Multiply(A=NamedRerouteDeclaration(None=@n210) {name=GlowMask_Hand_L, nodeColor=<rgba 0.5469 0 1 1>}, B=Power(Base=@n216.R, Exp=@n217))) {name=LeftHandRays, nodeColor=<rgba 1 0 0.4922 1>}
  TextureCoordinate()
 ]
}
