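# Parent character material for the female DenimEquip SuperLevel skin, adding parameter-driven hair recoloring to the 2023 character material stack.
# PreFX output optionally blends into the FN Hair 2023 function by hair mask, with a switchable fly-away hair pass.
# Base color becomes three hue, desaturation and brightness variants: Color01 and Color02 blend by powered mask red, then against Color03 by powered Fresnel.
# The result feeds the PostFX function.
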Material /BRCosmetics/Characters/Player/Female/Medium/Bodies/F_MED_DenimEquip_Stock/Skins/SuperLevel/Materials/M_DenimEquip_SL_Hair_PARENT
graph {
 properties {
  bUseMaterialAttributes=true
  bUsedWithInstancedStaticMeshes=true
  bUsedWithMeshDeformer=true
  bUsedWithSkeletalMesh=true
  pixelDepthOffsetMode=PDOM_Legacy
  shadingModel=MSM_FromMaterialExpression
  usedShadingModels=Default Lit | Subsurface | Clear Coat | Subsurface Profile | Cloth
 }
 nodes [
  n1 = MaterialFunctionCall() {materialFunction=&MF_CharacterParent_2023_PreFX}
  n2 = MaterialFunctionCall!(In=@n1.Result, Mask Input=@n1.Result) {materialFunction=/Game/Global/Materials/Character_Parent_2023/Hair_Parent_2023/MaterialFunctions/MF_FN_Hair_2023}
  n3 = BlendMaterialAttributes(
   A=@n1.Result
   B=StaticSwitchParameter(True=@n2.Result, False=@n1.Result) {defaultValue=true, group=Skin FX, parameterName=UseFNHair2_0}
   Alpha=@n2.HairMask
  )
  n4 = GetMaterialAttributes(None=StaticSwitchParameter(
    True=MaterialFunctionCall!(MaterialInput=@n3) {materialFunction=/Game/Global/Materials/Character_Parent_2023/Hair_Parent_2023/MaterialFunctions/MF_FlyAwayHair_2023}.Result
    False=@n3
   ) {group=Skin FX, parameterName=IsFlyAwayHair?}) {attributeGetTypes [#1]}
  n5 = ScalarParameter() {group=SUPERLEVELS, parameterName=Hair_Color03_HueShift}
  n6 = Multiply(
   A=Desaturation(
    None=MaterialFunctionCall(Hue Shift Percentage=@n5, Texture=@n4.Base Color) {materialFunction=&HueShift}.Result
    Fraction=ScalarParameter() {group=SUPERLEVELS, parameterName=Hair_Color03_Desaturation}
   )
   B=ScalarParameter() {defaultValue=1, group=SUPERLEVELS, parameterName=Hair_Color03_Brightness}
  )
  n7 = MaterialFunctionCall(Hue Shift Percentage=ScalarParameter() {group=SUPERLEVELS, parameterName=Hair_Color01_HueShift}, Texture=@n4.Base Color) {materialFunction=&HueShift}
  n8 = ScalarParameter() {group=SUPERLEVELS, parameterName=Hair_Color01_Desaturation}
  n9 = MaterialFunctionCall(Hue Shift Percentage=ScalarParameter() {group=SUPERLEVELS, parameterName=Hair_Color02_HueShift}, Texture=@n4.Base Color) {materialFunction=&HueShift}
  n10 = ScalarParameter() {group=SUPERLEVELS, parameterName=Hair_Color02_Desaturation}
  n11 = TextureSampleParameter2D() {
   group=00_Texture Inputs
   parameterName=Mask
   samplerType=SAMPLERTYPE_Masks
   sortPriority=3
   texture=&T_BlackMask
  }
  n12 = LinearInterpolate(
   A=Multiply(
    A=Desaturation(None=@n7.Result, Fraction=@n8)
    B=ScalarParameter() {defaultValue=1, group=SUPERLEVELS, parameterName=Hair_Color01_Brightness}
   )
   B=Multiply(
    A=ScalarParameter() {defaultValue=1, group=SUPERLEVELS, parameterName=Hair_Color02_Brightness}
    B=Desaturation(None=@n9.Result, Fraction=@n10)
   )
   Alpha=Power(Base=@n11.R, Exp=ScalarParameter() {defaultValue=8, group=SUPERLEVELS, parameterName=Hair_MaskPower, sortPriority=30}) {constExponent=8}
  )
  n13 = Power(Base=@n1.Fresnel, Exp=ScalarParameter() {group=SUPERLEVELS, parameterName=Hair_Color03_Power})
  MaterialFunctionCall(
   Material=SetMaterialAttributes!(MaterialAttributes=@n4.MaterialAttributes, Base Color=LinearInterpolate(A=@n6, B=@n12, Alpha=@n13), Metallic) {attributeSetTypes [#1, #2]}
   M=@n1.M
   Pre Skinned Local Position=@n1.M
   Distance Blend=@n1.M
   Fresnel=@n1.M
  ) {materialFunction=&MF_CharacterParent_2023_PostFX}
  Reroute(None=@n1.Tech Art Mask)
 ]
}
